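# Physics asset for the M_MED_Carabus_Full backpack cosmetic, with capsule bodies for its dangling belts, pack and wearer's body.
# Four belt chains (A and B, left and right) each link eleven dyn_belt bodies back to pack with tight swing limits.
# Twist limits narrow down each chain, and left and right sides use different limits.
# Pelvis, spine, upper arm, leg and dyn_Carabus bodies have no constraints.
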PhysicsAsset /BRCosmetics/Accessories/FORT_Backpacks/M_MED_Carabus_Full/Meshes/M_MED_Carabus_Full_Pack_Physics
bodies {
 calf_l [{shapeType=Capsule, length=37.0782, radius=10.7815, center=<-23.1858 0.0003 0.327>, rotation=<rot -89.1675 -179.9964 179.9961>}]
 calf_r [{shapeType=Capsule, length=36.4511, radius=11.4086, center=<23.941 0.0002 0.3267>, rotation=<rot -89.2184 -0.0011 0.0011>}]
 dyn_Carabus [{shapeType=Capsule, length=22.1294, radius=9.5014, center=<-0.329 -9.2301 -12.7872>, rotation=<rot -89.206 179.9992 -179.9991>}]
 dyn_CarabusFive [{shapeType=Capsule, length=34.2156, radius=33.5486, center=<-0.5346 -32.0691 6.3208>, rotation=<rot 0.9532 0.0002 -2.5414>}]
 dyn_CarabusFour [{shapeType=Capsule, length=23.1147, radius=12.8782, center=<-0.2017 -12.5308 -11.2728>, rotation=<rot -89.4029 -180 -180>}]
 dyn_CarabusSix [{shapeType=Capsule, length=31.7474, radius=36.6634, center=<-0.4415 -34.0535 4.9136>, rotation=<rot 0.1477 0 1.2482>}]
 dyn_beltA_l_01 [{shapeType=Capsule, length=10.6279, radius=1.2621, center=<-0.0067 -0.0904 -4.1493>, rotation=<rot 0 88.8747 -1.8591>}]
 dyn_beltA_l_010 [{shapeType=Capsule, length=17.7974, radius=1.8313, center=<0 -0.086 -9.1147>, rotation=<rot 0.0002 -90 0>}]
 dyn_beltA_l_011 [{shapeType=Capsule, length=17.9933, radius=1.8313, center=<0 0.4769 -6.8748>, rotation=<rot 0 179.9996 174.5182>}]
 dyn_beltA_l_02 [{shapeType=Capsule, length=8.014, radius=1.3395, center=<-0.0108 -0.1413 -3.8128>, rotation=<rot 2.5553 -90.0145 -0.3098>}]
 dyn_beltA_l_03 [{shapeType=Capsule, length=7.4881, radius=1.4364, center=<-0.0266 0.1316 -3.5297>, rotation=<rot 1.0395 -89.9952 -0.0778>}]
 dyn_beltA_l_04 [{shapeType=Capsule, length=7.2755, radius=1.5542, center=<0.0345 -0.1637 -4.5682>, rotation=<rot 0.7264 -90 0.4129>}]
 dyn_beltA_l_05 [{shapeType=Capsule, length=10.5213, radius=1.3021, center=<0.073 0.1074 -5.0727>, rotation=<rot -1.6065 -89.9999 -0.2014>}]
 dyn_beltA_l_06 [{shapeType=Capsule, length=9.7352, radius=1.2328, center=<0.1548 0.3679 -4.0089>, rotation=<rot -0.9075 -89.9997 -2.2994>}]
 dyn_beltA_l_07 [{shapeType=Capsule, length=8.2374, radius=1.5835, center=<0.1653 0.6364 -3.8124>, rotation=<rot -1.1671 -90 0.6364>}]
 dyn_beltA_l_08 [{shapeType=Capsule, length=13.2601, radius=1.8313, center=<0 0.086 -6.7605>, rotation=<rot -0.0001 -90 0>}]
 dyn_beltA_l_09 [{shapeType=Capsule, length=17.3466, radius=1.8313, center=<0 -0.0063 -8.626>, rotation=<rot -0.292 -90 0>}]
 dyn_beltA_r_01 [{shapeType=Capsule, length=10.5559, radius=0.9598, center=<-0.0321 -0.0028 4.2872>, rotation=<rot -0.7309 -89.8409 -2.3906>}]
 dyn_beltA_r_010 [{shapeType=Capsule, length=17.7974, radius=1.8313, center=<0 0.0273 9.7637>, rotation=<rot 1.1638 -90 0>}]
 dyn_beltA_r_011 [{shapeType=Capsule, length=16.5679, radius=2.5087, center=<-0.0001 0.4065 7.6089>, rotation=<rot 0 0 174.5165>}]
 dyn_beltA_r_02 [{shapeType=Capsule, length=7.8937, radius=1.4598, center=<-0.189 0.1034 4.0049>, rotation=<rot 2.8862 -90.0162 -0.3098>}]
 dyn_beltA_r_03 [{shapeType=Capsule, length=6.9984, radius=1.7098, center=<-0.0779 0.1081 3.4248>, rotation=<rot -0.093 -89.9935 -0.0778>}]
 dyn_beltA_r_04 [{shapeType=Capsule, length=6.0298, radius=1.9598, center=<-0.0581 -0.0275 3.7287>, rotation=<rot 2.1722 -90.0532 -0.2866>}]
 dyn_beltA_r_05 [{shapeType=Capsule, length=9.2051, radius=1.7589, center=<-0.098 -0.0783 4.1342>, rotation=<rot -2.0853 -90 -0.1574>}]
 dyn_beltA_r_06 [{shapeType=Capsule, length=9.9701, radius=1.5491, center=<-0.1521 -0.3355 3.6665>, rotation=<rot -4.4126 -90 -1.4618>}]
 dyn_beltA_r_07 [{shapeType=Capsule, length=8.2928, radius=1.5082, center=<-0.1344 -0.6433 4.144>, rotation=<rot -2.0896 -90 -0.1456>}]
 dyn_beltA_r_08 [{shapeType=Capsule, length=13.9262, radius=1.8313, center=<0 -0.0307 6.7004>, rotation=<rot 0.2001 -90 0>}]
 dyn_beltA_r_09 [{shapeType=Capsule, length=17.3466, radius=1.8313, center=<0 0.0701 7.2209>, rotation=<rot -0.4678 -90 0>}]
 dyn_beltB_l_01 [{shapeType=Capsule, length=4.7531, radius=2.5894, center=<-0.1822 0.0936 -3.0824>, rotation=<rot 7.2018 -90 -10.0143>}]
 dyn_beltB_l_010 [{shapeType=Capsule, length=12.2051, radius=2.4277, center=<0 0 -6.3652>, rotation=<rot 0 -90 0>}]
 dyn_beltB_l_011 [{shapeType=Capsule, length=16.6924, radius=2.9277, center=<0.0659 0.1143 -9.3022>, rotation=<rot -0.0001 -173.6239 -2.1027>}]
 dyn_beltB_l_02 [{shapeType=Capsule, length=3.7989, radius=2.6261, center=<0.3364 -0.2879 -2.7936>, rotation=<rot 0 -90.0038 -173.9985>}]
 dyn_beltB_l_03 [{shapeType=Capsule, length=3.2815, radius=3.2271, center=<-0.3274 0.0885 -3.1611>, rotation=<rot -6.8001 91.8788 -1.1416>}]
 dyn_beltB_l_04 [{shapeType=Capsule, length=4.6755, radius=3.0497, center=<0.0888 -0.1932 -3.9051>, rotation=<rot -0.3091 -89.9999 1.737>}]
 dyn_beltB_l_05 [{shapeType=Capsule, length=4.7994, radius=3.193, center=<0.1437 0.0675 -3.6115>, rotation=<rot -1.6967 -90 2.0581>}]
 dyn_beltB_l_06 [{shapeType=Capsule, length=4.5742, radius=3.2492, center=<0.1741 0.3603 -3.7035>, rotation=<rot -1.4807 -89.9998 3.2512>}]
 dyn_beltB_l_07 [{shapeType=Capsule, length=7.201, radius=2.886, center=<0.3311 0.635 -2.9446>, rotation=<rot -0.7624 -89.9999 -1.2526>}]
 dyn_beltB_l_08 [{shapeType=Capsule, length=12.0268, radius=2.5399, center=<0.0326 0.0402 -4.1898>, rotation=<rot -0.4815 -89.9999 -0.0024>}]
 dyn_beltB_l_09 [{shapeType=Capsule, length=9.6294, radius=2.2774, center=<0.0395 0.2209 -6.0369>, rotation=<rot -1.155 -90 0.0728>}]
 dyn_beltB_r_01 [{shapeType=Capsule, length=5.214, radius=2.3786, center=<0.2239 -0.1235 2.846>, rotation=<rot 7.2018 -90 -10.0143>}]
 dyn_beltB_r_010 [{shapeType=Capsule, length=12.0684, radius=2.5645, center=<-0.064 -0.0076 7.1854>, rotation=<rot -0.0861 -90 -0.0466>}]
 dyn_beltB_r_011 [{shapeType=Capsule, length=17.7682, radius=1.8518, center=<-0.0659 -0.1143 9.3022>, rotation=<rot 0 -173.6265 -2.1022>}]
 dyn_beltB_r_02 [{shapeType=Capsule, length=4.004, radius=2.4092, center=<-0.3607 0.288 2.4925>, rotation=<rot 0 -90.0326 -174.2849>}]
 dyn_beltB_r_03 [{shapeType=Capsule, length=3.7423, radius=2.7663, center=<0.2568 -0.0907 3.1576>, rotation=<rot 0 91.7438 -1.7531>}]
 dyn_beltB_r_04 [{shapeType=Capsule, length=5.4432, radius=2.6271, center=<-0.124 0.2325 3.4892>, rotation=<rot -4.8808 -90 7.4989>}]
 dyn_beltB_r_05 [{shapeType=Capsule, length=5.3239, radius=2.5537, center=<-0.0647 -0.0773 3.7826>, rotation=<rot -0.1832 -89.9999 -5.3817>}]
 dyn_beltB_r_06 [{shapeType=Capsule, length=5.3513, radius=2.5972, center=<-0.1504 -0.3376 3.7093>, rotation=<rot -4.6205 -89.9998 0.5768>}]
 dyn_beltB_r_07 [{shapeType=Capsule, length=8.3003, radius=2.6269, center=<-0.1961 -0.5662 3.4347>, rotation=<rot -4.8766 -89.9999 1.3817>}]
 dyn_beltB_r_08 [{shapeType=Capsule, length=13.2959, radius=2.3718, center=<-0.031 -0.0613 5.0736>, rotation=<rot 0.0389 -90 -0.0606>}]
 dyn_beltB_r_09 [{shapeType=Capsule, length=11.5461, radius=2.476, center=<-0.0534 0.0117 7.5199>, rotation=<rot -0.0857 -89.9999 -0.0467>}]
 pack [{shapeType=Capsule, length=21.8511, radius=13.5072, center=<0.2591 -8.4984 6.0096>, rotation=<rot 0 90.0818 -179.5299>}]
 pelvis [{shapeType=Capsule, length=13, radius=17}]
 spine_01 [{shapeType=Capsule, length=15.5, radius=14.5, center=<7.0336 0 0.0003>}]
 spine_02 [{shapeType=Capsule, length=16.75, radius=13.25, center=<15.1685 0 0.0005>}]
 thigh_l [{shapeType=Capsule, length=33.25, radius=10.75, center=<-23.5023 0.0001 -0.189>, rotation=<rot 89.4684 -180 180>}]
 thigh_r [{shapeType=Capsule, length=33.25, radius=10.75, center=<23.9814 0 0.1934>, rotation=<rot 89.5232 -179.9998 180>}]
 upperarm_l [{shapeType=Capsule, length=24.25, radius=5.75, center=<16.8806 0.0001 -0.2243>, rotation=<rot -89.2109 -179.9995 179.9995>}]
 upperarm_r [{shapeType=Capsule, length=24.25, radius=5.75, center=<-16.5493 -0.0003 0.2223>, rotation=<rot -89.2094 -179.9993 179.9993>}]
}
constraints [
 {bone1Name=dyn_beltA_l_01, bone2Name=pack, swing1LimitDegrees=2.5, swing2LimitDegrees=6, twistLimitDegrees=40}
 {bone1Name=dyn_beltA_l_02, bone2Name=dyn_beltA_l_01, swing1LimitDegrees=2.5, swing2LimitDegrees=6, twistLimitDegrees=20}
 {bone1Name=dyn_beltA_l_03, bone2Name=dyn_beltA_l_02, swing1LimitDegrees=2.5, swing2LimitDegrees=6, twistLimitDegrees=10}
 {bone1Name=dyn_beltB_l_01, bone2Name=pack, swing1LimitDegrees=2.5, swing2LimitDegrees=6, twistLimitDegrees=40}
 {bone1Name=dyn_beltB_l_02, bone2Name=dyn_beltB_l_01, swing1LimitDegrees=2.5, swing2LimitDegrees=6, twistLimitDegrees=20}
 {bone1Name=dyn_beltB_l_03, bone2Name=dyn_beltB_l_02, swing1LimitDegrees=2.5, swing2LimitDegrees=6, twistLimitDegrees=10}
 {bone1Name=dyn_beltA_r_01, bone2Name=pack, swing1LimitDegrees=5, swing2LimitDegrees=5, twistLimitDegrees=40}
 {bone1Name=dyn_beltA_r_02, bone2Name=dyn_beltA_r_01, swing1LimitDegrees=5, swing2LimitDegrees=5, twistLimitDegrees=35}
 {bone1Name=dyn_beltA_r_03, bone2Name=dyn_beltA_r_02, swing1LimitDegrees=5, swing2LimitDegrees=5, twistLimitDegrees=30}
 {bone1Name=dyn_beltB_r_01, bone2Name=pack, swing1LimitDegrees=5, swing2LimitDegrees=5, twistLimitDegrees=40}
 {bone1Name=dyn_beltB_r_02, bone2Name=dyn_beltB_r_01, swing1LimitDegrees=5, swing2LimitDegrees=5, twistLimitDegrees=35}
 {bone1Name=dyn_beltB_r_03, bone2Name=dyn_beltB_r_02, swing1LimitDegrees=5, swing2LimitDegrees=5, twistLimitDegrees=30}
 {bone1Name=dyn_beltB_r_04, bone2Name=dyn_beltB_r_03, swing1LimitDegrees=5, swing2LimitDegrees=5, twistLimitDegrees=25}
 {bone1Name=dyn_beltB_r_05, bone2Name=dyn_beltB_r_04, swing1LimitDegrees=5, swing2LimitDegrees=5, twistLimitDegrees=20}
 {bone1Name=dyn_beltB_r_06, bone2Name=dyn_beltB_r_05, swing1LimitDegrees=5, swing2LimitDegrees=5, twistLimitDegrees=20}
 {bone1Name=dyn_beltB_r_07, bone2Name=dyn_beltB_r_06, swing1LimitDegrees=5, swing2LimitDegrees=5, twistLimitDegrees=20}
 {bone1Name=dyn_beltA_r_07, bone2Name=dyn_beltA_r_06, swing1LimitDegrees=5, swing2LimitDegrees=5, twistLimitDegrees=20}
 {bone1Name=dyn_beltA_r_06, bone2Name=dyn_beltA_r_05, swing1LimitDegrees=5, swing2LimitDegrees=5, twistLimitDegrees=20}
 {bone1Name=dyn_beltA_r_05, bone2Name=dyn_beltA_r_04, swing1LimitDegrees=5, swing2LimitDegrees=5, twistLimitDegrees=20}
 {bone1Name=dyn_beltA_r_04, bone2Name=dyn_beltA_r_03, swing1LimitDegrees=5, swing2LimitDegrees=5, twistLimitDegrees=25}
 {bone1Name=dyn_beltB_l_07, bone2Name=dyn_beltB_l_06, swing1LimitDegrees=2.5, swing2LimitDegrees=6, twistLimitDegrees=10}
 {bone1Name=dyn_beltB_l_06, bone2Name=dyn_beltB_l_05, swing1LimitDegrees=2.5, swing2LimitDegrees=6, twistLimitDegrees=10}
 {bone1Name=dyn_beltB_l_05, bone2Name=dyn_beltB_l_04, swing1LimitDegrees=2.5, swing2LimitDegrees=6, twistLimitDegrees=10}
 {bone1Name=dyn_beltB_l_04, bone2Name=dyn_beltB_l_03, swing1LimitDegrees=2.5, swing2LimitDegrees=6, twistLimitDegrees=10}
 {bone1Name=dyn_beltA_l_07, bone2Name=dyn_beltA_l_06, swing1LimitDegrees=2.5, swing2LimitDegrees=6, twistLimitDegrees=10}
 {bone1Name=dyn_beltA_l_06, bone2Name=dyn_beltA_l_05, swing1LimitDegrees=2.5, swing2LimitDegrees=6, twistLimitDegrees=10}
 {bone1Name=dyn_beltA_l_05, bone2Name=dyn_beltA_l_04, swing1LimitDegrees=2.5, swing2LimitDegrees=6, twistLimitDegrees=10}
 {bone1Name=dyn_beltA_l_04, bone2Name=dyn_beltA_l_03, swing1LimitDegrees=2.5, swing2LimitDegrees=6, twistLimitDegrees=10}
 {bone1Name=dyn_beltA_l_08, bone2Name=dyn_beltA_l_07, swing1LimitDegrees=2.5, swing2LimitDegrees=6, twistLimitDegrees=10}
 {bone1Name=dyn_beltA_l_09, bone2Name=dyn_beltA_l_08, swing1LimitDegrees=2.5, swing2LimitDegrees=6, twistLimitDegrees=10}
 {bone1Name=dyn_beltA_l_010, bone2Name=dyn_beltA_l_09, swing1LimitDegrees=2.5, swing2LimitDegrees=6, twistLimitDegrees=10}
 {bone1Name=dyn_beltA_l_011, bone2Name=dyn_beltA_l_010, swing1LimitDegrees=2.5, swing2LimitDegrees=6, twistLimitDegrees=10}
 {bone1Name=dyn_beltB_l_08, bone2Name=dyn_beltB_l_07, swing1LimitDegrees=2.5, swing2LimitDegrees=6, twistLimitDegrees=10}
 {bone1Name=dyn_beltB_l_09, bone2Name=dyn_beltB_l_08, swing1LimitDegrees=2.5, swing2LimitDegrees=6, twistLimitDegrees=10}
 {bone1Name=dyn_beltB_l_010, bone2Name=dyn_beltB_l_09, swing1LimitDegrees=2.5, swing2LimitDegrees=6, twistLimitDegrees=10}
 {bone1Name=dyn_beltB_l_011, bone2Name=dyn_beltB_l_010, swing1LimitDegrees=2.5, swing2LimitDegrees=6, twistLimitDegrees=10}
 {bone1Name=dyn_beltA_r_08, bone2Name=dyn_beltA_r_07, swing1LimitDegrees=5, swing2LimitDegrees=5, twistLimitDegrees=20}
 {bone1Name=dyn_beltA_r_09, bone2Name=dyn_beltA_r_08, swing1LimitDegrees=5, swing2LimitDegrees=5, twistLimitDegrees=20}
 {bone1Name=dyn_beltA_r_010, bone2Name=dyn_beltA_r_09, swing1LimitDegrees=5, swing2LimitDegrees=5, twistLimitDegrees=20}
 {bone1Name=dyn_beltA_r_011, bone2Name=dyn_beltA_r_010, swing1LimitDegrees=5, swing2LimitDegrees=5, twistLimitDegrees=20}
 {bone1Name=dyn_beltB_r_08, bone2Name=dyn_beltB_r_07, swing1LimitDegrees=5, swing2LimitDegrees=5, twistLimitDegrees=20}
 {bone1Name=dyn_beltB_r_09, bone2Name=dyn_beltB_r_08, swing1LimitDegrees=5, swing2LimitDegrees=5, twistLimitDegrees=20}
 {bone1Name=dyn_beltB_r_010, bone2Name=dyn_beltB_r_09, swing1LimitDegrees=5, swing2LimitDegrees=5, twistLimitDegrees=20}
 {bone1Name=dyn_beltB_r_011, bone2Name=dyn_beltB_r_010, swing1LimitDegrees=5, swing2LimitDegrees=5, twistLimitDegrees=20}
]
