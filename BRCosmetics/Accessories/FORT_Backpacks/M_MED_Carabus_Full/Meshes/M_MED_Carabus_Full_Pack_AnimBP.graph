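# Animation Blueprint for the Carabus Full backpack, simulating its physics and toggling four pack elements to match the equipped pickaxe.
# Parent root, pelvis and lower-spine transforms are copied in world space each update, feeding RigidBody simulation with skydive and front-end tuning.
# Sprinting blends in a second RigidBody pass before a final pack offset.
# Weapon, skydive and emote events set material hide parameters and bone offsets from pickaxe meta tags.
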AnimBlueprint /BRCosmetics/Accessories/FORT_Backpacks/M_MED_Carabus_Full/Meshes/M_MED_Carabus_Full_Pack_AnimBP
target_skeleton=/Game/Accessories/FORT_Backpacks/M_MED_Carabus_Full/Meshes/M_MED_Carabus_Full_Pack_Skeleton
graphs [
 AnimGraph {nodes [
   "Constraint - pack (RR)"!(ComponentPose, Alpha=1, ConstraintWeights_0=1, ConstraintWeights_1:float=0) -> (Pose) {
    alphaBoolBlend {blendOption=Linear}
    boneToModify {boneName=pack}
    constraintSetup [{transformType=Rotation, targetBone {boneName=spine_05}}, {transformType=Rotation, targetBone {boneName=root}}]
    constraintWeights [1,1]
   }
   Copy Pose From Mesh() {bUseAttachedParent=true}
   n1 = RigidBody!(
    ComponentPose=Local To Component(LocalPose=Use cached pose 'initCache'())
    Alpha
    OverrideWorldGravity:Vector="0,0,-650"
    ExternalForce
    ComponentLinearAccScale:Vector="0.3,0.3,0.3"
    ComponentLinearVelScale:Vector="0.35,0.35,0.35"
   ) -> (Pose) {
    bClampLinearTranslationLimitToRefPose=true
    bOverrideWorldGravity=true
    overridePhysicsAsset=&/M_MED_Carabus_Full_Pack_Physics
    simulationSpace=BaseBoneSpace
    alphaBoolBlend {blendOption=Linear}
    baseBoneRef {boneName=pack}
    componentAppliedLinearAccClamp=<2000 2000 2000>
    overrideWorldGravity {z=-700}
   }
   n2 = Component To Local(ComponentPose=@n1)
   n3 = Use cached pose 'initCache'()
   n4 = Get bIsSprinting()
   n5 = "Transform (Modify) Bone - Bone: pack"(
    ComponentPose=Local To Component(LocalPose=Blend Poses by bool(BlendPose_0=@n2, BlendPose_1=@n3, BlendTime_0=0.5, BlendTime_1=0.5, bActiveValue=@n4) {blendType=Linear})
    Translation="0,0,1.5"
    Rotation="0,0,0"
    Scale="0.85,0.85,0.85"
   ) {
    scaleMode=BMM_Additive
    scaleSpace=BCS_BoneSpace
    translationMode=BMM_Additive
    translationSpace=BCS_BoneSpace
    alphaBoolBlend {blendOption=Linear}
    alphaScaleBiasClamp {bInterpResult=true, inRange {max=0, min=-40}, outRange {max=0, min=1}}
    boneToModify {boneName=pack}
    scale=<0.85 0.85 0.85>
    translation {z=1.5}
   }
   Output Pose(Result=Component To Local(ComponentPose=@n5)) {layerGroup=None}
   n6 = Get bIsInFrontEnd()
   n7 = OR Boolean!(A=Get bIsSkydiveFloating(), B=Get bIsSkyDiving(), C:bool) -> (ReturnValue)
   n8 = RigidBody(
    ComponentPose=Local To Component(LocalPose=Use cached pose 'initCache1'())
    OverrideWorldGravity=SelectVector(A="10,0,200", B=SelectVector(A="0,0,-980", B="0,0,-780", bPickA=@n6), bPickA=@n7)
    ComponentLinearAccScale=SelectVector(A="0.4,0.4,0.4", B=SelectVector(A="0.05,0.05,0.05", B="0.3,0.35,0.35", bPickA=@n6), bPickA=@n7)
    ComponentLinearVelScale="0.35,0.35,0.35"
   ) {
    bClampLinearTranslationLimitToRefPose=true
    bOverrideWorldGravity=true
    overridePhysicsAsset=&M_MED_Carabus_Full_Pack_Physics
    simulationSpace=BaseBoneSpace
    alphaBoolBlend {blendOption=Linear}
    baseBoneRef {boneName=pack}
    componentAppliedLinearAccClamp=<1200 1200 1200>
    overrideWorldGravity {z=-700}
   }
   Save cached pose 'initCache'(Pose=Component To Local(ComponentPose=@n8)) {cacheName=initCache}
   n9 = "Transform (Modify) Bone - Bone: dyn_Carabus"(
    ComponentPose=Local To Component(LocalPose=Local Space Ref Pose())
    Alpha=1
    Translation=Get WeaponSwitch01() -> (WeaponSwitch01:Vector)
    Rotation="0,0,0"
    Scale="0,0,0"
   ) {
    translationMode=BMM_Additive
    translationSpace=BCS_BoneSpace
    alphaBoolBlend {blendOption=Linear}
    alphaScaleBiasClamp {bInterpResult=true, inRange {max=0, min=-40}, outRange {max=0, min=1}}
    boneToModify {boneName=dyn_Carabus}
    scale=<0 0 0>
    translation {z=1.5}
   }
   n10 = "Transform (Modify) Bone - Bone: dyn_CarabusFour"(ComponentPose=@n9, Alpha=1, Translation=Get WeaponSwitch02() -> (WeaponSwitch02:Vector), Rotation="0,0,0", Scale="0,0,0") {
    translationMode=BMM_Additive
    translationSpace=BCS_BoneSpace
    alphaBoolBlend {blendOption=Linear}
    alphaScaleBiasClamp {bInterpResult=true, inRange {max=0, min=-40}, outRange {max=0, min=1}}
    boneToModify {boneName=dyn_CarabusFour}
    scale=<0 0 0>
    translation {z=1.5}
   }
   n11 = "Transform (Modify) Bone - Bone: dyn_CarabusFive"(ComponentPose=@n10, Alpha=1, Translation=Get WeaponSwitch03() -> (WeaponSwitch03:Vector), Rotation="0,0,0", Scale="0,0,0") {
    translationMode=BMM_Additive
    translationSpace=BCS_BoneSpace
    alphaBoolBlend {blendOption=Linear}
    alphaScaleBiasClamp {bInterpResult=true, inRange {max=0, min=-40}, outRange {max=0, min=1}}
    boneToModify {boneName=dyn_CarabusFive}
    scale=<0 0 0>
    translation {z=1.5}
   }
   n12 = "Transform (Modify) Bone - Bone: dyn_CarabusSix"(ComponentPose=@n11, Alpha=1, Translation=Get WeaponSwitch04() -> (WeaponSwitch04:Vector), Rotation="0,0,0", Scale="0,0,0") {
    translationMode=BMM_Additive
    translationSpace=BCS_BoneSpace
    alphaBoolBlend {blendOption=Linear}
    alphaScaleBiasClamp {bInterpResult=true, inRange {max=0, min=-40}, outRange {max=0, min=1}}
    boneToModify {boneName=dyn_CarabusSix}
    scale=<0 0 0>
    translation {z=1.5}
   }
   n13 = "Transform (Modify) Bone - Bone: body_root"(ComponentPose=@n12, Alpha=1, Translation=Get root_tran() -> (root_tran:Vector), Rotation=Get root_rot() -> (root_rot:Rotator), Scale="1,1,1") {
    rotationMode=BMM_Replace
    rotationSpace=BCS_WorldSpace
    scaleSpace=BCS_BoneSpace
    translationMode=BMM_Replace
    translationSpace=BCS_WorldSpace
    alphaBoolBlend {blendOption=Linear}
    alphaScaleBiasClamp {bInterpResult=true, inRange {max=0, min=-40}, outRange {max=0, min=1}}
    boneToModify {boneName=body_root}
   }
   n14 = "Transform (Modify) Bone - Bone: pelvis"(
    ComponentPose=@n13
    Alpha=1
    Translation=Get pelvis_tran() -> (pelvis_tran:Vector)
    Rotation=Get pelvis_rot() -> (pelvis_rot:Rotator)
    Scale="1,1,1"
   ) {
    rotationMode=BMM_Replace
    rotationSpace=BCS_WorldSpace
    scaleSpace=BCS_BoneSpace
    translationMode=BMM_Replace
    translationSpace=BCS_WorldSpace
    alphaBoolBlend {blendOption=Linear}
    alphaScaleBiasClamp {inRange {max=0, min=-40}, outRange {max=0, min=1}}
    boneToModify {boneName=pelvis}
   }
   n15 = "Transform (Modify) Bone - Bone: spine_01"(
    ComponentPose=@n14
    Alpha=1
    Translation=Get spine01_tran() -> (spine01_tran:Vector)
    Rotation=Get spine01_rot() -> (spine01_rot:Rotator)
    Scale="1,1,1"
   ) {
    rotationMode=BMM_Replace
    rotationSpace=BCS_WorldSpace
    scaleSpace=BCS_BoneSpace
    translationMode=BMM_Replace
    translationSpace=BCS_WorldSpace
    alphaBoolBlend {blendOption=Linear}
    alphaScaleBiasClamp {inRange {max=0, min=-40}, outRange {max=0, min=1}}
    boneToModify {boneName=spine_01}
   }
   n16 = "Transform (Modify) Bone - Bone: spine_02"(
    ComponentPose=@n15
    Alpha=1
    Translation=Get spine02_tran() -> (spine02_tran:Vector)
    Rotation=Get spine02_rot() -> (spine02_rot:Rotator)
    Scale="1,1,1"
   ) {
    rotationMode=BMM_Replace
    rotationSpace=BCS_WorldSpace
    scaleSpace=BCS_BoneSpace
    translationMode=BMM_Replace
    translationSpace=BCS_WorldSpace
    alphaBoolBlend {blendOption=Linear}
    alphaScaleBiasClamp {inRange {max=0, min=-40}, outRange {max=0, min=1}}
    boneToModify {boneName=spine_02}
   }
   Save cached pose 'initCache1'(Pose=Component To Local(ComponentPose=@n16)) {cacheName=initCache1}
  ]}
 EventGraph {nodes [
   Event OnWeaponChanged()
   > n1 = Sequence!(execute) -> (then_0, then_1, then_2:Exec, then_3:Exec)
   > n2 = Get Dynamic Material Instance(SkeletalMeshComponent=GetOwningComponent(), ElementIndex=1)
   n3 = PickaxeItemDefinitionHasAnyMetaTag(Tags=Make Array("[0]"=(TagName="Cosmetics.MetaTags.CarabusFull")), bRequiresEquipped=true)
   > n4 = Branch(Condition=@n3)
   > n5 = Branch(Condition=Get bIsSkyDiving())
   n6 = Get Dynamic Material Instance(execute=@n35.then_0, SkeletalMeshComponent=GetOwningComponent(), ElementIndex=1)
   SetScalarParameterValue(execute=[@n22.then_0, @n28.then_0, @n4.else, @n24.else, @n5.then], self=[@n2, @n6], ParameterName=Hide Element 01, Value=0)
   > Set WeaponSwitch01(WeaponSwitch01:Vector="0, 0, 0") -> (Output_Get:Vector)
   SetScalarParameterValue(execute=[@n24.then, @n5.else], self=[@n2, @n6], ParameterName=Hide Element 01, Value=1)
   > Set WeaponSwitch01(WeaponSwitch01="0, 0,40")
   n7 = Get Dynamic Material Instance(execute=@n1.then_1, SkeletalMeshComponent=GetOwningComponent(), ElementIndex=1)
   n8 = PickaxeItemDefinitionHasAnyMetaTag(Tags=Make Array("[0]"=(TagName="Cosmetics.MetaTags.CarabusFullFour")), bRequiresEquipped=true)
   > n9 = Branch(Condition=@n8)
   > n10 = Branch(Condition=Get bIsSkyDiving())
   n11 = Get Dynamic Material Instance(execute=@n35.then_1, SkeletalMeshComponent=GetOwningComponent(), ElementIndex=1)
   SetScalarParameterValue(execute=[@n22.then_1, @n28.then_1, @n9.else, @n25.else, @n10.then], self=[@n7, @n11], ParameterName=Hide Element 02, Value=0)
   > Set WeaponSwitch02(WeaponSwitch02:Vector="0, 0, 0") -> (Output_Get:Vector)
   SetScalarParameterValue(execute=[@n25.then, @n10.else], self=[@n7, @n11], ParameterName=Hide Element 02, Value=1)
   > Set WeaponSwitch02(WeaponSwitch02="0, 0,40")
   n12 = Get Dynamic Material Instance(execute=@n1.then_2, SkeletalMeshComponent=GetOwningComponent(), ElementIndex=1)
   n13 = PickaxeItemDefinitionHasAnyMetaTag(Tags=Make Array("[0]"=(TagName="Cosmetics.MetaTags.CarabusFullFive")), bRequiresEquipped=true)
   > n14 = Branch(Condition=@n13)
   > n15 = Branch(Condition=Get bIsSkyDiving())
   n16 = Get Dynamic Material Instance(execute=@n35.then_2, SkeletalMeshComponent=GetOwningComponent(), ElementIndex=1)
   SetScalarParameterValue(execute=[@n22.then_2, @n28.then_2, @n14.else, @n26.else, @n15.then], self=[@n12, @n16], ParameterName=Hide Element 03, Value=0)
   > Set WeaponSwitch03(WeaponSwitch03:Vector="0, 0,-10") -> (Output_Get:Vector)
   SetScalarParameterValue(execute=[@n26.then, @n15.else], self=[@n12, @n16], ParameterName=Hide Element 03, Value=1)
   > Set WeaponSwitch03(WeaponSwitch03="0, 0,20")
   n17 = Get Dynamic Material Instance(execute=@n1.then_3, SkeletalMeshComponent=GetOwningComponent(), ElementIndex=1)
   n18 = PickaxeItemDefinitionHasAnyMetaTag(Tags=Make Array("[0]"=(TagName="Cosmetics.MetaTags.CarabusFullSix")), bRequiresEquipped=true)
   > n19 = Branch(Condition=@n18)
   > n20 = Branch(Condition=Get bIsSkyDiving())
   n21 = Get Dynamic Material Instance(execute=@n35.then_3, SkeletalMeshComponent=GetOwningComponent(), ElementIndex=1)
   SetScalarParameterValue(execute=[@n22.then_3, @n28.then_3, @n19.else, @n20.then, @n27.else], self=[@n17, @n21], ParameterName=Hide Element 04, Value=0)
   > Set WeaponSwitch04(WeaponSwitch04:Vector="0, 0, 0") -> (Output_Get:Vector)
   SetScalarParameterValue(execute=[@n20.else, @n27.then], self=[@n17, @n21], ParameterName=Hide Element 04, Value=1)
   > Set WeaponSwitch04(WeaponSwitch04="0, 0,60")
   Event OnBeginSkydiving()
   > n22 = Sequence()
   Event OnEndSkydiving()
   > n23 = Sequence()
   n24 = Branch(execute=[@n6.then, @n23.then_0], Condition=@n3)
   n25 = Branch(execute=[@n11.then, @n23.then_1], Condition=@n8)
   n26 = Branch(execute=[@n16.then, @n23.then_2], Condition=@n13)
   n27 = Branch(execute=[@n21.then, @n23.then_3], Condition=@n18)
   Event OnBeginPlayingEmote() {is=title K2Node_Event}
   > n28 = Sequence()
   Event BlueprintUpdateAnimation()
   n29 = GetPartSkeletalMeshComponent()
   > Is Valid(InputObject=@n29)
   n30 = GetAttachParent!(self:Scene Component Object Reference=@n29) -> (ReturnValue:Scene Component Object Reference)
   > Is Valid(InputObject=@n30)
   n31 = BreakTransform(InTransform=GetSocketTransform(self=@n30, InSocketName=root, TransformSpace=RTS_World))
   > Set root_tran(root_tran:Vector=@n31) -> (Output_Get:Vector)
   > Set root_rot(root_rot:Rotator=@n31.Rotation) -> (Output_Get:Rotator)
   n32 = BreakTransform(InTransform=GetSocketTransform(self=@n30, InSocketName=pelvis, TransformSpace=RTS_World))
   > Set pelvis_tran(pelvis_tran:Vector=@n32) -> (Output_Get:Vector)
   > Set pelvis_rot(pelvis_rot:Rotator=@n32.Rotation) -> (Output_Get:Rotator)
   n33 = BreakTransform(InTransform=GetSocketTransform(self=@n30, InSocketName=spine_01, TransformSpace=RTS_World))
   > Set spine01_tran(spine01_tran:Vector=@n33) -> (Output_Get:Vector)
   > Set spine01_rot(spine01_rot:Rotator=@n33.Rotation) -> (Output_Get:Rotator)
   n34 = BreakTransform(InTransform=GetSocketTransform(self=@n30, InSocketName=spine_02, TransformSpace=RTS_World))
   > Set spine02_tran(spine02_tran:Vector=@n34) -> (Output_Get:Vector)
   > Set spine02_rot(spine02_rot:Rotator=@n34.Rotation) -> (Output_Get:Rotator)
   Event OnEndPlayingEmote() {is=title K2Node_Event}
   > n35 = Sequence()
   TryGetPawnOwner()
   TryGetPawnOwner()
  ]}
]
variables [
 pelvis_rot {type=Rotator}
 pelvis_tran {type=Vector}
 root_rot {type=Rotator}
 root_tran {type=Vector}
 spine01_rot {type=Rotator}
 spine01_tran {type=Vector}
 spine02_rot {type=Rotator}
 spine02_tran {type=Vector}
 WeaponSwitch01 {type=Vector}
 WeaponSwitch02 {type=Vector}
 WeaponSwitch03 {type=Vector}
 WeaponSwitch04 {type=Vector}
]
